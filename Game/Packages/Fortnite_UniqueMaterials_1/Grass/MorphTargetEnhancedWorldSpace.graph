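# Material function in the Grass package that morphs static meshes by time of day and outputs both the offset and the shifted world position.
# It sums the red, blue and alpha channels of FortniteActiveTimeOfDay, adds MorphAmount and clamps the result.
# That weight scales morph target 1's WPO from StaticMeshMorphTargets.
# A static switch can disable morphing, returning zero offset and unmodified world position.
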MaterialFunction /Game/Packages/Fortnite_UniqueMaterials_1/Grass/MorphTargetEnhancedWorldSpace
graph {nodes [
  n1 = CollectionParameter() {
   collection=&FortniteMaterialParameters
   parameterName=FortniteActiveTimeOfDay
  }
  n2 = DotProduct(A=@n1, B=Constant4Vector() {constant=<rgba 1 0 1 1>}.RGBA)
  n3 = FunctionInput() {bUsePreviewValueAsDefault=true, inputName=MorphAmount, inputType=FunctionInput_Scalar}
  n4 = Multiply(
   A=Clamp(None=Add(A=@n2, B=@n3))
   B=MaterialFunctionCall() {materialFunction=/Engine/Functions/Engine_MaterialFunctions02/WorldPositionOffset/StaticMeshMorphTargets}.Morph Target 1 WPO
  )
  n5 = FunctionInput(Preview=StaticBool() {value=true}) {bUsePreviewValueAsDefault=true, inputName=Use Morph Targets, inputType=FunctionInput_StaticBool, previewValue=<0.25 0 0 1>}
  FunctionOutput(None=StaticSwitch(True=@n4, False=Constant(), Value=@n5)) {outputName=MorphTarget Info}
  n6 = WorldPosition()
  FunctionOutput(None=StaticSwitch(True=Add(A=@n4, B=@n6.XYZ), False=@n6.XYZ, Value=@n5)) {outputName=Modified World Position}
 ]}
