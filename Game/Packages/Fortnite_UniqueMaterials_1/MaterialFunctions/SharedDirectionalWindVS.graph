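# Shared material function for global directional foliage wind on grass or trees, outputting wind amount, wind vector and a pivot sphere mask.
# Pivot-based (grass) or world-position UVs, aligned to the collection-driven wind direction, pan across WindFlowMap.
# Two flow-map octaves form the wind amount; a larger third drives root rotation in the displacement chain.
# That chain is disconnected: the WPO output is a zero constant, so it moves no vertices.
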MaterialFunction /Game/Packages/Fortnite_UniqueMaterials_1/MaterialFunctions/SharedDirectionalWindVS
graph {
 properties {bExposeToLibrary=true, description=Global directional wind function for grass or trees}
 nodes [
  FunctionOutput(None=Constant()) {outputName=WPO}
  n1 = MaterialFunctionCall!(Wind Vector, Optimized for Foliage Placement, Optimized Vector) {materialFunction=/Engine/Functions/Engine_MaterialFunctions02/PivotPainter/PivotPainter_PerObjectFoliageData}
  n2 = MaterialFunctionCall(f3=MaterialFunctionCall() {materialFunction=/Engine/Functions/Engine_MaterialFunctions02/PivotPainter/PivotPainter_PerObjectData}.Pivot Position) {materialFunction=&AddComponents}
  n3 = FunctionInput(Preview=ScalarParameter() {defaultValue=30.8902, parameterName="Grass Pivot Offset "}) {bUsePreviewValueAsDefault=true, inputName=Grass Pivot Animation Offset1, inputType=FunctionInput_Scalar}
  n4 = FunctionInput(Preview=Constant3Vector() {constant=<rgba 0 0 1 0>}) {bUsePreviewValueAsDefault=true, inputName=Element Vector}
  n5 = DotProduct(A=WorldPosition().XYZ, B=@n4)
  n6 = FunctionInput(Preview=ScalarParameter() {defaultValue=1.6178, parameterName=Length Wise Animation Offset}) {bUsePreviewValueAsDefault=true, inputName=Length Wise Animation Offset1, inputType=FunctionInput_Scalar}
  n7 = Divide(
   A=Add(A=Multiply(A=@n2.f3, B=@n3), B=Multiply(A=@n5, B=@n6))
   B=ScalarParameter() {defaultValue=3634.3567, group=directional wind, parameterName=Wind Texture Scale}
  )
  n8 = CollectionParameter() {
   collection=&FortniteMaterialParameters
   parameterName=WindDirection_RGB Speed_A
  }
  n9 = FunctionInput(Preview=MaterialFunctionCall(Vector=ComponentMask(None=@n8) {b=true, g=true, r=true}, Default=Constant3Vector() {constant=<rgba 1 0 0 0>}) {materialFunction=&SafeNormalize}.Result) {bUsePreviewValueAsDefault=true, inputName=Wind Vector}
  n10 = FunctionInput(Preview=CrossProduct(A=@n9, B=@n4)) {bUsePreviewValueAsDefault=true, inputName=Force Vector}
  n11 = FunctionInput(Preview=MaterialFunctionCall() {materialFunction=&PivotPainter_PerObjectData}.Pivot Position) {bUsePreviewValueAsDefault=true, inputName=grass Pivot}
  n12 = Divide(A=DotProduct(A=@n11, B=@n4)) {constB=-3}
  n13 = DotProduct(A=@n11, B=@n9)
  n14 = Divide(
   A=AppendVector(A=DotProduct(A=@n10, B=@n11), B=Add(A=@n12, B=@n13))
   B=ScalarParameter() {defaultValue=3634.3567, group=directional wind, parameterName=Wind Texture Scale}
  )
  n15 = WorldPosition()
  n16 = DotProduct(A=@n15.XYZ, B=@n10)
  n17 = ScalarParameter() {defaultValue=1.6178, parameterName=Tree Length Wise Animation Offset}
  n18 = Add(
   A=Multiply(
    A=DotProduct(A=@n15.XYZ, B=@n4)
    B=FunctionInput(Preview=@n17) {bUsePreviewValueAsDefault=true, inputName=Length Wise Animation Offset, inputType=FunctionInput_Scalar}
   )
   B=DotProduct(A=@n15.XYZ, B=@n9)
  )
  n19 = StaticSwitch(
   True=Add(A=Multiply(A=@n1.Random Value Per Element) {constB=0.1}, B=Add(A=@n7, B=@n14))
   False=Divide(A=AppendVector(A=@n16, B=@n18), B=ScalarParameter() {defaultValue=3634.3567, group=directional wind, parameterName=Wind Texture Scale})
   Value=FunctionInput(Preview=StaticBool()) {bUsePreviewValueAsDefault=true, inputName=Grass/Flowers?, inputType=FunctionInput_StaticBool}
  )
  n20 = Multiply(
   A=Time()
   B=ComponentMask(None=VectorParameter() {group=directional wind, parameterName=Wind Panning Speed, defaultValue=<rgba 0.3869 -0.602 0 1>}.RGB) {g=true, r=true}
  )
  n21 = TextureSample(UVs=Frac(None=Add(A=@n19, B=@n20))) {samplerType=SAMPLERTYPE_LinearColor, texture=/Game/Global/Textures/Default/Flowmaps/WindFlowMap}
  n22 = MaterialFunctionCall(Float2=ConstantBiasScale(None=@n21.RGB) {bias=-0.5, scale=2}) {materialFunction=&BreakOutFloat2Components}
  n23 = Multiply(A=@n22.R, B=@n10)
  n24 = Multiply(A=@n22.G, B=@n9)
  n25 = Add(A=@n19, B=Multiply(A=@n20) {constB=1.5})
  n26 = Frac(None=Multiply(A=Add(A=@n25) {constB=0.3458}, B=ScalarParameter() {defaultValue=-0.0378, group=directional wind, parameterName=Second Octave Scale}) {constB=3})
  n27 = MaterialFunctionCall(Float2=ConstantBiasScale(None=TextureSample(UVs=@n26) {samplerType=SAMPLERTYPE_LinearColor, texture=&WindFlowMap}.RGB) {bias=-0.5, scale=2}) {materialFunction=&BreakOutFloat2Components}
  n28 = Add(A=Multiply(A=@n27.R, B=@n9), B=Multiply(A=@n27.G, B=@n10))
  n29 = ScalarParameter() {defaultValue=0.05, parameterName=Second Octave Influence}
  n30 = Add(A=@n29)
  n31 = ScalarParameter() {defaultValue=0.7248, group=directional wind, parameterName=Directional Offset Amount}
  n32 = Add(
   A=Add(A=Add(A=@n23, B=@n24), B=Multiply(A=@n28, B=@n29))
   B=Multiply(
    A=Multiply(A=@n9, B=@n30)
    B=FunctionInput(Preview=@n31) {bUsePreviewValueAsDefault=true, inputName=Directional Offset Amount, inputType=FunctionInput_Scalar}
   )
  )
  n33 = ConstantBiasScale(None=DotProduct(A=@n32, B=@n9)) {bias=1.25}
  FunctionOutput(None=@n33) {outputName=Black and white wind amount}
  FunctionOutput(None=@n9) {outputName=Wind Vector}
  n34 = OneMinus(None=SphereMask(
    A=MaterialFunctionCall() {materialFunction=&ComponentPivotLocation}.Result
    B=WorldPosition().XYZ
    Radius=ObjectRadius()
   ) {hardnessPercent=0})
  n35 = Multiply(A=@n34, B=@n34)
  FunctionOutput(None=@n35) {outputName=sphere mask}
  SphereMask(
   A=MaterialFunctionCall!(WindVector, Max Dist for Parent Piv) {materialFunction=/Engine/Functions/Engine_MaterialFunctions02/PivotPainter/PivotPainter_TreeData}.Leaf Piv Position
   B=WorldPosition().XYZ
   Radius=FunctionInput(Preview=ScalarParameter() {defaultValue=276.2969, group=directional wind, parameterName=Directional Wind Falloff From Leaf Pivot Position Radius}) {bUsePreviewValueAsDefault=true, inputName=Directional Wind Falloff From Leaf Pivot Position Radius}
  ) {attenuationRadius=60, hardnessPercent=0}
  n36 = CollectionParameter() {
   collection=&FortniteMaterialParameters
   parameterName=Foliage Directional Wind Intensity
  }
  StaticSwitchParameter(True=Divide(A=@n36) {constB=4}, False=@n36) {group=directional wind, parameterName=ThrottleWindIIntensity}
  n37 = ScalarParameter() {defaultValue=0.0066, group=directional wind, parameterName=Large Rotation Octave Scale}
  n38 = Multiply(A=@n20, B=@n37) {constB=1.5}
  n39 = Frac(None=Multiply(A=Add(A=@n19, B=@n38), B=@n37) {constB=3})
  n40 = MaterialFunctionCall(Float2=ConstantBiasScale(None=TextureSample(UVs=@n39) {samplerType=SAMPLERTYPE_LinearColor, texture=&WindFlowMap}.RGB) {bias=-0.5, scale=2}) {materialFunction=&BreakOutFloat2Components}
  n41 = Add(A=Multiply(A=@n40.R, B=@n9), B=Multiply(A=@n40.G, B=@n10))
  n42 = CollectionParameter() {
   collection=&FortniteMaterialParameters
   parameterName=Foliage Directional Wind Intensity
  }
  n43 = Multiply(
   A=Multiply(A=DotProduct(A=@n41, B=@n9), B=ScalarParameter() {defaultValue=-0.05, parameterName=Rotation Amount Directional Wind})
   B=StaticSwitchParameter(True=Divide(A=@n42) {constB=4}, False=@n42) {group=directional wind, parameterName=ThrottleWindIIntensity}
  )
  n44 = MaterialFunctionCall() {materialFunction=&ComponentPivotLocation}
  n45 = LinearInterpolate(
   A=Multiply(A=@n33, B=@n9)
   B=@n32
   Alpha=ScalarParameter() {defaultValue=0.5, group=directional wind, parameterName=Random Direction Influence}
  )
  n46 = Multiply(A=@n21.RGB, B=ScalarParameter() {defaultValue=410.7188, group=directional wind, parameterName=Strength})
  n47 = FunctionInput(Preview=@n35) {bUsePreviewValueAsDefault=true, inputName=Additional Motion Mask, inputType=FunctionInput_Scalar, previewValue=<1 0 0 1>}
  n48 = Multiply(A=Multiply(A=Multiply(A=@n45, B=@n46), B=@n36), B=@n47)
  n49 = Add(A=@n48, B=WorldPosition().XYZ)
  StaticSwitch(
   True=Add(A=RotateAboutAxis(NormalizedRotationAxis=@n10, RotationAngle=@n43, PivotPoint=@n44.Result, Position=@n49), B=@n48)
   False=@n48
   Value=FunctionInput(Preview=StaticBool() {value=true}) {bUsePreviewValueAsDefault=true, inputName=Use Root Rotation, inputType=FunctionInput_StaticBool}
  )
  StaticSwitch()
 ]
}
