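# Masked two-sided foliage master material for instanced and Nanite grass, with wind sway, player push and optional world-space color variation.
# Pivot Painter or morph-target pivots anchor blade rotation, driven by layered wind, gust and lull sines with root falloff.
# Vertex alpha masks that motion plus optional distance morph and shrink, then shared directional wind is added.
# Normals bend toward local up near pivots, and Debug switches default to camera-based push.
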Material /Game/Packages/Fortnite_UniqueMaterials_1/Grass/Fortress_Grass
graph {
 properties {
  bEnableResponsiveAA=true
  bUseMaterialAttributes=true
  bUsedWithInstancedStaticMeshes=true
  bUsedWithNanite=true
  blendMode=BLEND_Masked
  pixelDepthOffsetMode=PDOM_Legacy
  refractionMethod=RM_IndexOfRefraction
  shadingModel=MSM_TwoSidedFoliage
  twoSided=true
  usedShadingModels=Two Sided Foliage
 }
 nodes [
  DistanceCullFade()
  n1 = CollectionParameter() {
   collection=&FortniteMaterialParameters
   parameterName=WindDirection_RGB Speed_A
  }
  n2 = MaterialFunctionCall!(
   Wind Vector=@n1
   Optimized for Foliage Placement=StaticBoolParameter() {defaultValue=true, group=Foliage, parameterName=Optimized for foliage placement}
   Optimized Vector=VectorParameter() {group=Foliage, parameterName=Optimized for foliage painter vector, defaultValue=<rgba 0 0 1 1>}.RGB
  ) {materialFunction=/Engine/Functions/Engine_MaterialFunctions02/PivotPainter/PivotPainter_PerObjectFoliageData}
  n3 = StaticSwitch(
   True=MaterialFunctionCall() {materialFunction=/Engine/Functions/Engine_MaterialFunctions02/WorldPositionOffset/StaticMeshMorphTargets}.Pivot Position WS
   False=@n2.Piv Position
   Value=StaticBoolParameter() {group=MorphTarget, parameterName=Use Morph Target Batch Script Pivot Location}
  )
  n4 = DotProduct(A=ComponentMask(None=@n1) {b=true, g=true, r=true}, B=@n3)
  n5 = ComponentMask(None=Divide(A=@n4, B=ScalarParameter() {defaultValue=-600, group=DeadWave, parameterName="(VS) Dead Wave Width"})) {desc=Wind direction gradient, r=true}
  n6 = DotProduct(
   A=CrossProduct(A=ComponentMask(None=@n1) {b=true, desc=wind direc, g=true, r=true}, B=Constant3Vector() {constant=<rgba 0 0 1 0>}) {desc=Find Right angle}
   B=@n3
  )
  n7 = VectorParameter() {group=DeadWave, parameterName="(VS)DeadVerticalShiftWave2 R_Size(WS) G_Speed B_Strength", defaultValue=<rgba 512 -0.15 -25 0>}
  n8 = Divide(A=@n6, B=@n7.R)
  n9 = CollectionParameter() {collection=&FortniteMaterialParameters, parameterName=WindTimeAccumulator}
  n10 = StaticSwitchParameter(True=Multiply(A=@n2.Wind Accumulator) {constB=0.05}, False=@n9) {defaultValue=true, desc=wind time accumulator, group=Debug, parameterName=Debug}
  n11 = Multiply(A=@n10, B=@n7.G)
  n12 = Multiply(A=Sine(None=Add(A=@n8, B=@n11)), B=@n7.B) {desc=Vertical offset - sub wave}
  n13 = Multiply(A=@n10, B=ScalarParameter() {defaultValue=0.1, group=VertexShader, parameterName="(VS)Animation Speed Multiplier"})
  n14 = Add(
   A=StaticSwitchParameter(True=Constant(), False=Add(A=@n5, B=@n12)) {defaultValue=true, group=Optimization, parameterName=test Optimization}
   B=Multiply(A=@n13, B=ScalarParameter() {defaultValue=-2, group=DeadWave, parameterName=Dead Wave Speed Multiplier})
  ) {desc=add time to wp gradient}
  MaterialFunctionCall(Value=@n14, -1 to 1=StaticBool()) {materialFunction=&LinearSine}
  n15 = MaterialFunctionCall() {materialFunction=&StaticMeshMorphTargets}
  n16 = MaterialFunctionCall() {materialFunction=&PivotPainter_PerObjectFoliageData}
  n17 = StaticBoolParameter() {group=MorphTarget, parameterName=Use Morph Target Batch Script Pivot Location}
  n18 = Divide(
   A=ComponentMask(None=StaticSwitch(True=@n15.Pivot Position WS, False=@n16.Piv Position, Value=@n17)) {g=true, r=true}
   B=ScalarParameter() {defaultValue=128, group=Pixel_Diffuse, parameterName=Color Variation Texture UV WS Divisor}
  )
  n19 = Subtract(A=TextureSampleParameter2D(UVs=@n18) {group=Pixel_Diffuse, parameterName=Hue Variation WS texture, texture=/Game/Global/Textures/Default/Noise/Noise}.R) {constB=0.5}
  n20 = ScalarParameter() {defaultValue=0.1, group=Pixel_Diffuse, parameterName=HueVariationAmount}
  n21 = TextureSampleParameter2D() {group=Pixel_Diffuse, parameterName=Diffuse, texture=/Game/Packages/Fortress_Foliage/Textures/grass_fortnite_D}
  n22 = Multiply(A=@n21.RGB, B=VectorParameter() {group=Pixel_Diffuse, parameterName=DiffuseMod, defaultValue=<rgba 1 1 1 0>}.RGB)
  n23 = Add(
   A=MaterialFunctionCall(
    Base=@n22
    Blend=TextureSampleParameter2D(UVs=@n18) {group=Pixel_Diffuse, parameterName=ColorVariationMap, texture=/Game/Packages/Fortress_Foliage/Textures/GrassBladeColor}.RGB
   ) {materialFunction=&Blend_Overlay}.Result
   B=@n22
  )
  n24 = Constant() {r=0.12}
  n25 = StaticSwitchParameter(
   True=MaterialFunctionCall(Hue Shift Percentage=Multiply(A=@n19, B=@n20), Texture=@n22) {materialFunction=&HueShift}.Result
   False=StaticSwitchParameter(True=Desaturation(None=@n23, Fraction=@n24), False=@n22) {group=Pixel_Diffuse, parameterName=GrassPixelShader}
  ) {desc=d and ssc, group=Pixel_Diffuse, parameterName=RandomizeHue}
  n26 = MaterialFunctionCall(DiffuseColor=@n25, SpecularColor=Constant()) {materialFunction=&ConvertFromDiffSpec}
  n27 = TextureSampleParameter2D() {group=Specular, parameterName=Specular, samplerType=SAMPLERTYPE_Masks, texture=/Game/Environments/Props/Foliage/Textures/Mushroom01_S}
  n28 = LinearInterpolate(
   A=ScalarParameter() {group=Specular, parameterName=RoughnessLow}
   B=ScalarParameter() {defaultValue=1, group=Specular, parameterName=RoughnessHigh}
   Alpha=@n27.B
  )
  n29 = Constant() {r=1}
  n30 = Constant()
  n31 = Clamp(None=Divide(A=Distance(A=@n3, B=@n30), B=Constant() {r=256}))
  n32 = Constant()
  n33 = Constant() {r=1}
  n34 = StaticSwitchParameter(
   True=@n21.A
   False=TextureSampleParameter2D() {group=Pixel_Mask, parameterName=OpacityMask, texture=/Game/Characters/Fortress_Character/Textures/T_All_White}.R
  ) {defaultValue=true, group=Pixel_Mask, parameterName=Use Diffuse Alpha As Mask}
  n35 = TextureSampleParameter2D() {
   group=Pixel_Normal
   parameterName=Normal
   samplerType=SAMPLERTYPE_Normal
   texture=/Game/Environments/Props/Residential/Textures/Prop_Potted_Plant_N
  }
  n36 = StaticSwitchParameter(True=@n35.RGB, False=Constant3Vector() {constant=<rgba 0 0 1 0>}) {group=Pixel_Normal, parameterName=Use a normal map}
  n37 = Transform(None=Constant3Vector() {constant=<rgba 0 0 1 0>}) {transformSourceType=TRANSFORMSOURCE_Local, transformType=TRANSFORM_Tangent}
  n38 = SphereMask(
   A=MaterialFunctionCall() {materialFunction=&PivotPainter_PerObjectFoliageData}.Object PIvot Location
   B=WorldPosition().XYZ
  ) {attenuationRadius=75, hardnessPercent=0}
  n39 = VertexColor() {desc=masks animation}
  n40 = CollectionParameter() {
   collection=&FortniteMaterialParameters
   parameterName=PlayerCapsuleBottomPosition
  }
  n41 = StaticSwitchParameter(True=CameraPositionWS(), False=ComponentMask(None=@n40) {b=true, g=true, r=true}) {defaultValue=true, group=Debug, parameterName=Debug}
  n42 = Distance(A=@n3, B=@n41)
  n43 = ScalarParameter() {defaultValue=2048, group=CullEffect, parameterName=StartMorphDistance}
  n44 = Clamp(None=Divide(A=Subtract(A=@n42, B=@n43), B=ScalarParameter() {defaultValue=1024, group=CullEffect, parameterName=MorphDistance}))
  n45 = MaterialFunctionCall(
   Value=MaterialFunctionCall(Speed=Constant() {r=0.5}) {materialFunction=&TimeWithSpeedVariable}.Frac Time
   Sine Phase=StaticBool()
  ) {materialFunction=&LinearSine}
  n46 = Constant()
  n47 = MaterialFunctionCall!(
   Use Morph Targets=StaticBoolParameter() {group=MorphTarget, parameterName=Use Morph Target}
   MorphAmount=StaticSwitchParameter(
    True=Multiply(A=@n44, B=@n44)
    False=StaticSwitchParameter(True=@n45.Linear Sine, False=@n46) {group=MorphTarget, parameterName=Debug Morph Via Time}
   ) {group=MorphTarget, parameterName=Morph Based on distance}
  ) {materialFunction=&/MorphTargetEnhancedWorldSpace}
  n48 = Normalize(VectorInput=ComponentMask(None=@n1) {b=true, g=true, r=true}) {desc=Wind Vector}
  n49 = ComponentMask(None=@n2.Element X-Axis) {b=true, g=true, r=true}
  n50 = CrossProduct(A=@n48, B=@n49) {desc=Proper Wind Axis}
  n51 = CollectionParameter() {
   collection=&FortniteMaterialParameters
   parameterName=PlayerCapsuleBottomPosition
  }
  n52 = ComponentMask(None=@n51) {b=true, g=true, r=true}
  n53 = Subtract(A=@n52, B=@n3) {desc=- world piv pos}
  n54 = Distance(A=@n52, B=@n3) {desc=PerElementMagnitude}
  n55 = ScalarParameter() {defaultValue=100, group=PlayerPush, parameterName="(VS)PlayerPush Distance"}
  n56 = OneMinus(None=Clamp(None=Divide(A=@n54, B=@n55)))
  n57 = Clamp(None=Multiply(A=@n56, B=ScalarParameter() {defaultValue=100, desc=3, group=PlayerPush, parameterName="(VS)PlayerPush Falloff Lerp Multiplier"}) {constB=2})
  n58 = LinearInterpolate(
   A=@n50
   B=CrossProduct(
    A=MaterialFunctionCall(Vector=@n53) {materialFunction=&SafeNormalize}.Result
    B=Constant3Vector() {constant=<rgba 0 0.0001 1 0>}
   )
   Alpha=@n57
  )
  n59 = Multiply(
   A=ConstantBiasScale(None=DotProduct(A=@n48, B=@n49)) {bias=-1}
   B=ScalarParameter() {defaultValue=1, group=VertexShader, parameterName="(VS) Wind Bias Amount"}
  )
  n60 = Divide(A=@n4, B=ScalarParameter() {defaultValue=-600, group=gust, parameterName="(VS) Gust Wave Width"})
  n61 = VectorParameter() {group=gust, parameterName="(VS)GustVerticalShiftWave2 R_Size(WS) G_Speed B_Strength", defaultValue=<rgba 512 -0.15 -25 0>}
  n62 = Sine(None=Add(A=Divide(A=@n6, B=@n61.R), B=Multiply(A=@n10, B=@n61.G)))
  n63 = StaticSwitchParameter(
   True=Constant()
   False=Add(A=ComponentMask(None=@n60) {desc=Wind direction gradient, r=true}, B=Multiply(A=@n62, B=@n61.B) {desc=Vertical offset - sub wave})
  ) {defaultValue=true, group=Optimization, parameterName=test Optimization}
  n64 = Multiply(A=@n13, B=ScalarParameter() {defaultValue=-2, group=gust, parameterName=Gust Speed Multiplier})
  n65 = Clamp(None=Sine(None=Add(A=@n63, B=@n64) {desc=add time to wp gradient})) {minDefault=-0.25}
  n66 = Subtract(A=OneMinus(None=Multiply(A=@n65, B=@n65) {desc=.25}) {desc=.75}, B=Constant() {r=0.375})
  n67 = VectorParameter() {
   group=VertexShader
   parameterName="(VS)Animation Offsets R_VertAlpha G_DistanceCalculationFromRoot B_RandomNumber A_WorldPositionSineOffset"
   defaultValue=<rgba 0.1 -0.0125 0 0.0005>
  }
  n68 = Distance(A=@n47.Modified World Position, B=@n3) {desc=sub piv}
  n69 = ScalarParameter() {defaultValue=26, group=Foliage, parameterName="(VS)Average Object Size WS/Root stiffness"}
  n70 = MaterialFunctionCall() {materialFunction=/Engine/Functions/Engine_MaterialFunctions02/WorldPositionOffset/ObjectScale}
  n71 = Multiply(A=@n69, B=@n70.Scale X)
  n72 = Multiply(
   A=@n2.Uniform Object Scale
   B=StaticSwitchParameter(
    True=Divide(A=@n69, B=@n70.Scale X)
    False=StaticSwitchParameter(True=@n71, False=@n69) {defaultValue=true, group=Foliage, parameterName=Scale Average Object Size With Model Scale}
   ) {group=Foliage, parameterName=Inverse Scale Average Object Size With Model Scale}
  ) {desc=obj scale}
  n73 = MaterialFunctionCall() {materialFunction=&BoundingBoxBased_0-1_UVW}
  n74 = StaticSwitchParameter(
   True=Clamp(None=Divide(A=@n68, B=@n72))
   False=Multiply(
    A=ComponentMask(None=@n73.RGB) {b=true}
    B=ScalarParameter() {defaultValue=0.5, group=Foliage, parameterName=Z Bounding Box Falloff Multiplier}
   )
  ) {defaultValue=true, group=Foliage, parameterName=Use World Space Root Falloff Motion Limiting or if false use Bounding Box Z Falloff}
  n75 = Abs(None=@n74)
  n76 = Add(A=Multiply(A=@n2.Custom Alpha Values, B=@n67.R) {desc=Custom alpha settings from script}, B=Multiply(A=@n67.G, B=@n75) {desc=root})
  n77 = Multiply(A=@n2.Random Value Per Element, B=@n67.B) {desc=RandomValue 0-1 per element}
  n78 = ScalarParameter() {defaultValue=-3, group=VertexShader, parameterName="(VS) Main Wave Width"}
  n79 = VectorParameter() {group=VertexShader, parameterName="(VS)VerticalShiftWave R_Size(WS) G_Speed B_Strength", defaultValue=<rgba 2048 0.1 50 0>}
  n80 = Sine(None=Add(A=Divide(A=@n6, B=@n79.R), B=Multiply(A=@n10, B=@n79.G)))
  n81 = VectorParameter() {group=VertexShader, parameterName="(VS)VerticalShiftWave2 R_Size(WS) G_Speed B_Strength", defaultValue=<rgba 512 -0.15 -25 0>}
  n82 = Sine(None=Add(A=Divide(A=@n6, B=@n81.R), B=Multiply(A=@n10, B=@n81.G)))
  n83 = Add(
   A=ComponentMask(None=Divide(A=@n4, B=@n78)) {desc=Wind direction gradient, r=true}
   B=Add(A=Multiply(A=@n80, B=@n79.B) {desc=Vertical offset}, B=Multiply(A=@n82, B=@n81.B) {desc=Vertical offset - sub wave})
  )
  n84 = Add(A=@n13, B=Add(A=Add(A=@n76, B=@n77), B=Multiply(A=@n67.A, B=@n83) {desc=Pan Across World}) {desc=+ world}) {desc=Add Time}
  n85 = Add(
   A=Multiply(
    A=Add(A=@n59, B=@n66) {desc=Bias Motion Away from Wind Source}
    B=ScalarParameter() {defaultValue=0.5, group=gust, parameterName="(VS) Gust Animation Strength"}
   )
   B=Add(A=@n59, B=Sine(None=@n84) {desc=Sine, period=0.1}) {desc=Bias Motion Away from Wind Source}
  )
  n86 = Sine(None=@n14)
  n87 = Multiply(
   A=Add(A=ConstantBiasScale(None=@n86), B=ScalarParameter() {defaultValue=0.25, group=DeadWave, parameterName=DeadBias})
   B=ScalarParameter() {defaultValue=0.5, group=DeadWave, parameterName=DeadScale}
  )
  n88 = CollectionParameter() {collection=&FortniteMaterialParameters, parameterName=WindStrength}
  n89 = Multiply(
   A=Multiply(
    A=Multiply(A=@n85, B=@n87)
    B=StaticSwitchParameter(True=@n2.Wind Strength, False=@n88) {defaultValue=true, desc=wind strength, group=Debug, parameterName=Debug}
   ) {constA=1, desc=Wind Strength}
   B=Multiply(A=@n75, B=@n75) {desc="< anim strength - Root Falloff"}
  ) {desc=mask root motion}
  n90 = Multiply(
   A=Multiply(A=Multiply(A=@n56, B=@n56), B=ScalarParameter() {defaultValue=0.5, group=PlayerPush, parameterName="(VS)PlayerPush Bend Multiplier"})
   B=@n75
  ) {desc=Mask root motion - linear falloff}
  n91 = StaticSwitchParameter(True=LinearInterpolate(A=@n89, B=@n90, Alpha=@n57), False=Add(A=@n89, B=@n90)) {defaultValue=true, group=PlayerPush, parameterName=Lerp t or add player push animation f}
  n92 = Multiply(
   A=@n39.A
   B=RotateAboutAxis(
    NormalizedRotationAxis=StaticSwitchParameter(True=@n58, False=@n50) {defaultValue=true, group=PlayerPush, parameterName=UsePlayerPush}
    RotationAngle=StaticSwitchParameter(True=@n91, False=@n89) {defaultValue=true, group=PlayerPush, parameterName=UsePlayerPush}
    PivotPoint=@n3
    Position=@n47.Modified World Position
   ) {desc=wo}
  )
  n93 = Distance(A=@n3, B=@n41)
  n94 = ScalarParameter() {defaultValue=2048, group=CullEffect, parameterName=StartShrinkingDistance}
  n95 = Clamp(None=Divide(A=Subtract(A=@n93, B=@n94), B=ScalarParameter() {defaultValue=1024, group=CullEffect, parameterName=ShrinkingDistance}))
  n96 = Multiply(A=@n95, B=@n95)
  n97 = Add(
   A=@n92
   B=Multiply(
    A=Subtract(A=@n3, B=@n47.Modified World Position) {desc=pivot position - modified world position}
    B=StaticSwitchParameter(True=@n96, False=@n95) {group=MorphTarget, parameterName=Use Morph Target}
   )
  )
  n98 = Multiply(
   A=@n39.A
   B=Add(A=@n47.MorphTarget Info, B=StaticSwitchParameter(True=@n97, False=@n92) {group=VertexShader, parameterName=ScaleOverDistance})
  ) {desc=delete eventually removes scaling as well}
  n99 = MaterialFunctionCall(Float3=TransformPosition(None=WorldPosition().XYZ) {transformSourceType=TRANSFORMPOSSOURCE_World}) {materialFunction=&BreakOutFloat3Components}
  n100 = ScalarParameter() {defaultValue=10, parameterName=Directional Wind Mask Height}
  n101 = MaterialFunctionCall!(
   Force Vector
   Grass/Flowers?=StaticBool() {value=true}
   Directional Wind Falloff From Leaf Pivot Position Radius=ScalarParameter() {defaultValue=360.7985, group=directional wind, parameterName=Directional Wind Falloff From Leaf Pivot Position Radius}
   Additional Motion Mask=Clamp(None=Divide(A=@n99.B, B=@n100) {constB=60})
   grass Pivot=@n3
   Directional Offset Amount
   Use Root Rotation=StaticBool()
   Wind Vector
   Element Vector
   Length Wise Animation Offset
   Length Wise Animation Offset1
   Grass Pivot Animation Offset1
  ) {materialFunction=/Game/Packages/Fortnite_UniqueMaterials_1/MaterialFunctions/SharedDirectionalWindVS}
  MaterialFunctionCall!(
   Material attributes=MakeMaterialAttributes!(
    BaseColor=@n26.BaseColor
    Metallic=StaticSwitchParameter(True=@n27.G, False=@n26.Metallic) {group=Specular, parameterName=UseSpecMask}
    Specular=StaticSwitchParameter(True=@n27.R, False=@n26.Specular) {group=Specular, parameterName=UseSpecMask}
    Roughness=StaticSwitchParameter(True=@n28, False=@n29) {group=Specular, parameterName=UseSpecMask}
    Anisotropy
    EmissiveColor=StaticSwitchParameter(True=@n31, False=@n32) {group=Debug, parameterName=debugPivot}
    Opacity=ScalarParameter() {defaultValue=0.5, group=Pixel_SSS, parameterName=SSS Amount}
    OpacityMask=StaticSwitchParameter(True=@n33, False=@n34) {group=Debug, parameterName=debugPixelShader}
    Normal=LinearInterpolate(A=@n36, B=@n37, Alpha=@n38)
    Tangent
    WorldPositionOffset=Add(A=@n98, B=@n101.WPO)
    SubsurfaceColor=@n25
    ClearCoat
    ClearCoatRoughness
    AmbientOcclusion
    Refraction
    CustomizedUVs_0
    CustomizedUVs_1
    CustomizedUVs_2
    CustomizedUVs_3
    CustomizedUVs_4
    CustomizedUVs_5
    CustomizedUVs_6
    CustomizedUVs_7
    PixelDepthOffset
    ShadingModel
    Displacement
   )
   IsGameplayAsset
  ) {materialFunction=/Game/Packages/Fortress_SharedMaterials/Base_Material_Functions/Fortnite_BaseFunction}
  Multiply()
  n102 = WorldPosition() {worldPositionShaderOffset=WPT_ExcludeAllShaderOffsets}
  n103 = DotProduct(B=@n102.XYZ) {desc=wind vector}
  n104 = ComponentMask(None=CrossProduct(B=@n102.XYZ)) {b=true}
  n105 = Multiply(A=Divide(A=AppendVector(A=@n103, B=@n104), B=Constant() {r=1800}), B=Constant2Vector() {g=1, r=0.5})
  n106 = Time()
  n107 = AppendVector(A=Multiply(A=@n106, B=Constant() {r=-0.125}), B=Multiply(A=@n106, B=Constant() {r=0.05}))
  MaterialExpressionObjectWorldPosition_2 = ObjectPositionWS()
  n108 = DotProduct(
   A=Subtract(A=WorldPosition() {worldPositionShaderOffset=WPT_ExcludeAllShaderOffsets}.XYZ, B=@MaterialExpressionObjectWorldPosition_2)
   B=ObjectOrientation()
  )
  n109 = Constant() {r=1024}
  n110 = TextureSample(UVs=Add(A=Add(A=@n105, B=@n107), B=Divide(A=@n108, B=@n109))) {texture=/Game/Packages/Fortress_Foliage/Materials/grassNormals}
  n111 = Constant() {desc=strength, r=0.25}
  Normalize(VectorInput=Add(A=Multiply(A=@n110.RGB, B=@n111), B=Constant3Vector() {constant=<rgba 0 0 1 0>}))
 ]
}
